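# Social panel confirmation button widget, derived from CommonButtonLegacy, showing a label with a colored accept or decline indicator strip and a hover bubble.
# PreConstruct sets the text from DisplayText and picks the indicator color from bIsAccept.
# Hover plays the OnHover animation, and unhover reverses it at double speed.
# That animation slides the label, animates border material parameters, and reveals the Select input prompt and check icon.
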WidgetBlueprint /Game/UI/SocialPanel/SocialConfirmationButton
animations [OnHover {
  binding_count=7
  display_label=""
  length_seconds=0.2
  object=&self:OnHover
  sequence {bindings [
    HoverBubble {
     id=#1
     possessed_class=SizeBox
     tracks [{
       type=MovieScene2DTransformTrack
       sections [{
         type=MovieScene2DTransformSection
         end_frame=4
         start_frame=0
         channels [
          Translation.X {type=MovieSceneScriptingFloatChannel, keys [[0,0],[2,8]]}
          Translation.Y {type=MovieSceneScriptingFloatChannel}
          Angle {type=MovieSceneScriptingFloatChannel}
          Scale.X {type=MovieSceneScriptingFloatChannel}
          Scale.Y {type=MovieSceneScriptingFloatChannel}
          Shear.X {type=MovieSceneScriptingFloatChannel}
          Shear.Y {type=MovieSceneScriptingFloatChannel}
         ]
        }]
      }]
    }
    InteractionIcon {
     id=#2
     possessed_class=Image
     tracks [
      {
       type=MovieScene2DTransformTrack
       sections [{
         type=MovieScene2DTransformSection
         end_frame=4
         start_frame=0
         channels [
          Translation.X {type=MovieSceneScriptingFloatChannel}
          Translation.Y {type=MovieSceneScriptingFloatChannel}
          Angle {type=MovieSceneScriptingFloatChannel}
          Scale.X {type=MovieSceneScriptingFloatChannel, keys [[0,0.875],[3,1.125],[4,1]]}
          Scale.Y {type=MovieSceneScriptingFloatChannel, keys [[0,0.875],[3,1.125],[4,1]]}
          Shear.X {type=MovieSceneScriptingFloatChannel}
          Shear.Y {type=MovieSceneScriptingFloatChannel}
         ]
        }]
      }
      {
       type=MovieSceneFloatTrack
       display_name=Render Opacity
       sections [{type=MovieSceneFloatSection, end_frame=4, start_frame=0, channels [None {type=MovieSceneScriptingFloatChannel, keys [[0,0],[2,1]]}]}]
      }
     ]
    }
    EntryBorder {
     id=#3
     possessed_class=Border
     tracks [
      {
       type=MovieSceneWidgetMaterialTrack
       sections [{
         type=MovieSceneParameterSection
         end_frame=4
         start_frame=0
         channels [AnimateHighlight {type=MovieSceneScriptingFloatChannel, keys [[0,0],[3,1],[4,1]]}]
        }]
      }
      {
       type=MovieScene2DTransformTrack
       sections [{
         type=MovieScene2DTransformSection
         end_frame=4
         start_frame=0
         channels [
          Translation.X {type=MovieSceneScriptingFloatChannel, keys [[0,0],[2,-8]]}
          Translation.Y {type=MovieSceneScriptingFloatChannel}
          Angle {type=MovieSceneScriptingFloatChannel}
          Scale.X {type=MovieSceneScriptingFloatChannel}
          Scale.Y {type=MovieSceneScriptingFloatChannel}
          Shear.X {type=MovieSceneScriptingFloatChannel}
          Shear.Y {type=MovieSceneScriptingFloatChannel}
         ]
        }]
      }
     ]
    }
    InputActionWidget {
     id=#4
     possessed_class=CommonActionWidget
     tracks [{
       type=MovieSceneFloatTrack
       display_name=Render Opacity
       sections [{type=MovieSceneFloatSection, end_frame=4, start_frame=0, channels [None {type=MovieSceneScriptingFloatChannel, keys [[0,0],[1,0],[2,1]]}]}]
      }]
    }
    AnimatedInputBorder {
     id=#5
     possessed_class=Border
     tracks [
      {
       type=MovieSceneWidgetMaterialTrack
       sections [{type=MovieSceneParameterSection, channels [AnimateStateChange {type=MovieSceneScriptingFloatChannel, keys [[0,0],[3,1],[4,1]]}]}]
      }
      {
       type=MovieSceneColorTrack
       display_name=Brush Color
       sections [{
         type=MovieSceneColorSection
         channels [
          Color.R {type=MovieSceneScriptingFloatChannel}
          Color.G {type=MovieSceneScriptingFloatChannel}
          Color.B {type=MovieSceneScriptingFloatChannel}
          Color.A {type=MovieSceneScriptingFloatChannel, keys [[0,0],[2,1],[4,1]]}
         ]
        }]
      }
     ]
    }
    Text_InteractionName {
     id=#6
     possessed_class=CommonTextBlock
     tracks [
      {
       type=MovieSceneFloatTrack
       display_name=Render Opacity
       sections [{type=MovieSceneFloatSection, end_frame=4, start_frame=0, channels [None {type=MovieSceneScriptingFloatChannel, keys [[0,0.5],[4,1]]}]}]
      }
      {
       type=MovieScene2DTransformTrack
       sections [{
         type=MovieScene2DTransformSection
         end_frame=4
         start_frame=0
         channels [
          Translation.X {type=MovieSceneScriptingFloatChannel, keys [[0,0],[2,-8]]}
          Translation.Y {type=MovieSceneScriptingFloatChannel}
          Angle {type=MovieSceneScriptingFloatChannel}
          Scale.X {type=MovieSceneScriptingFloatChannel}
          Scale.Y {type=MovieSceneScriptingFloatChannel}
          Shear.X {type=MovieSceneScriptingFloatChannel}
          Shear.Y {type=MovieSceneScriptingFloatChannel}
         ]
        }]
      }
     ]
    }
    Border_InteractionIndicator {
     id=#7
     possessed_class=Border
     tracks [{
       type=MovieSceneFloatTrack
       display_name=Render Opacity
       sections [{type=MovieSceneFloatSection, end_frame=4, start_frame=0, channels [None {type=MovieSceneScriptingFloatChannel, keys [[0,0.3],[3,1],[4,1]]}]}]
      }]
    }
   ]}
 }]
graphs [EventGraph {nodes [
   Event PreConstruct()
   > "SetText (Text)"(self=Get Text_InteractionName() -> (Text_InteractionName:Common Text Object Reference), InText=Get DisplayText() -> (DisplayText:Text))
   > SetBrushColor(
    self=Get Border_InteractionIndicator() -> (Border_InteractionIndicator:Border Object Reference)
    InBrushColor=Select(
     Option 0:Linear Color Structure=(R=0.4287,G=0.6654,B=0.7991,A=1)
     Option 1:Linear Color Structure=(R=0.7605,G=0.0123,B=0.0482,A=1)
     Index=Get bIsAccept() -> (bIsAccept:bool)
    ) -> (ReturnValue:Linear Color Structure)
   )
   Event On Hovered()
   > PlayAnimation(InAnimation=Get OnHover() -> (OnHover:Widget Animation Object Reference))
   Event On Unhovered()
   > PlayAnimation(InAnimation=Get OnHover(), PlayMode=Reverse, PlaybackSpeed=2)
  ]}]
variables [
 bIsAccept {type=boolean}
 DisplayText {type=string, default=""}
]
widgets {
 info {inheritedWidgetCount=1, parentClass=&CommonButtonLegacy, rootWidgetClass=&SizeBox, widgetCount=16}
 widgets [
  SizeBox_0 {widgetClassPath=&SizeBox, widget=SizeBox_0, properties {bOverride_HeightOverride=true, heightOverride=48}}
  EntryBorder {
   widgetClassPath=&Border
   parent=SizeBox_0
   widget=EntryBorder
   properties {
    visibility=SelfHitTestInvisible
    background {
     bIsDynamicallyLoaded=false
     drawAs=Image
     imageType=NoImage
     mirroring=NoMirror
     resourceName=None
     resourceObject=/Game/UI/SocialPanel/Art/Materials/M_UI_SocialUserItemBorder_Interaction
     tiling=NoTile
     imageSize=<32 32>
     margin=<ltrb 0 0 0 0>
     outlineSettings {
      bUseBrushTransparency=false
      roundingType=HalfHeightRadius
      width=0
      color {colorUseRule=UseColor_Specified, specifiedColor=<rgba 0 0 0 0>}
      cornerRadii=<0 0 0 0>
     }
     tintColor {colorUseRule=UseColor_Specified, specifiedColor=<rgba 1 1 1 1>}
     uVRegion {bIsValid=false, max=<0 0>, min=<0 0>}
    }
    padding=<ltrb 0 0 0 0>
   }
   slot {class=SizeBoxSlot, properties {padding=<ltrb 8 0 8 0>}}
  }
  HorizontalBox_2 {
   widgetClassPath=&HorizontalBox
   parent=EntryBorder
   widget=HorizontalBox_2
   slot {class=BorderSlot, properties {padding=<ltrb 0 0 0 0>}}
  }
  Border_InteractionIndicator {
   widgetClassPath=&Border
   bIsVariable=true
   parent=HorizontalBox_2
   widget=Border_InteractionIndicator
   properties {renderOpacity=0.3, brushColor=<rgba 0.4287 0.6654 0.7991 1>, padding=<ltrb 0 0 0 0>}
   slot {class=HorizontalBoxSlot}
  }
  Spacer_0 {
   widgetClassPath=&Spacer
   parent=Border_InteractionIndicator
   widget=Spacer_0
   properties {size=<8 8>}
   slot {class=BorderSlot, properties {padding=<ltrb 0 0 0 0>}}
  }
  SizeBox_3 {
   widgetClassPath=&SizeBox
   parent=HorizontalBox_2
   widget=SizeBox_3
   properties {bOverride_MinDesiredHeight=true, heightOverride=24, minDesiredHeight=24}
   slot {
    class=HorizontalBoxSlot
    properties {horizontalAlignment=HAlign_Left, verticalAlignment=VAlign_Center, padding=<ltrb 24 0 0 0>, size {sizeRule=Fill, value=1}}
   }
  }
  Text_InteractionName {
   widgetClassPath=&CommonTextBlock
   bIsVariable=true
   parent=SizeBox_3
   widget=Text_InteractionName
   properties {
    mobileFontSizeMultiplier=1.2
    renderOpacity=0.5
    text=NO
    font {
     bForceMonospaced=false
     bMaterialIsStencil=false
     fontMaterial=null
     fontObject=&BurbankSmall
     letterSpacing=0
     monospacedWidth=1
     size=18
     skewAmount=0
     typefaceFontName=Black
     outlineSettings {
      bApplyOutlineToDropShadows=false
      bMiteredCorners=false
      bSeparateFillAlpha=false
      outlineMaterial=null
      outlineSize=0
      outlineColor=<rgba 0 0 0 1>
     }
    }
   }
   slot {class=SizeBoxSlot, properties {horizontalAlignment=HAlign_Left, verticalAlignment=VAlign_Center, padding=<ltrb 0 -6 0 -8>}}
  }
  HoverBubble {
   widgetClassPath=&SizeBox
   parent=HorizontalBox_2
   widget=HoverBubble
   properties {bOverride_MinDesiredWidth=true, minDesiredWidth=80}
   slot {class=HorizontalBoxSlot, properties {horizontalAlignment=HAlign_Right}}
  }
  Overlay_1 {widgetClassPath=&Overlay, parent=HoverBubble, widget=Overlay_1, slot {class=SizeBoxSlot}}
  AnimatedInputBorder {
   widgetClassPath=&Border
   parent=Overlay_1
   widget=AnimatedInputBorder
   properties {
    visibility=SelfHitTestInvisible
    background {
     bIsDynamicallyLoaded=false
     drawAs=Image
     imageType=NoImage
     mirroring=NoMirror
     resourceName=None
     resourceObject=/Game/UI/SocialPanel/Art/Materials/M_UI_UserHover_Inst
     tiling=NoTile
     imageSize=<64 64>
     margin=<ltrb 1 0 0 0>
     outlineSettings {
      bUseBrushTransparency=false
      roundingType=HalfHeightRadius
      width=0
      color {colorUseRule=UseColor_Specified, specifiedColor=<rgba 0 0 0 0>}
      cornerRadii=<0 0 0 0>
     }
     tintColor {colorUseRule=UseColor_Specified, specifiedColor=<rgba 1 1 1 1>}
     uVRegion {bIsValid=false, max=<0 0>, min=<0 0>}
    }
    padding=<ltrb 0 0 0 0>
   }
   slot {class=OverlaySlot, properties {horizontalAlignment=HAlign_Right, verticalAlignment=VAlign_Fill}}
  }
  HorizontalBox_6 {
   widgetClassPath=&HorizontalBox
   parent=AnimatedInputBorder
   widget=HorizontalBox_6
   slot {class=BorderSlot, properties {padding=<ltrb 0 0 0 0>}}
  }
  InputActionWidget {
   widgetClassPath=&CommonActionWidget
   bInherited=true
   parent=HorizontalBox_6
   widget=InputActionWidget
   properties {visibility=SelfHitTestInvisible, inputActions [{dataTable=&InputActionDataTable, rowName=Select}]}
   slot {class=HorizontalBoxSlot, properties {horizontalAlignment=HAlign_Left, verticalAlignment=VAlign_Center, padding=<ltrb 8 0 -16 0>}}
  }
  Spacer_7 {
   widgetClassPath=&Spacer
   parent=HorizontalBox_6
   widget=Spacer_7
   properties {size=<8 8>}
   slot {class=HorizontalBoxSlot, properties {size {sizeRule=Fill, value=1}}}
  }
  InteractionIconSB {
   widgetClassPath=&SizeBox
   parent=HorizontalBox_6
   widget=InteractionIconSB
   properties {bOverride_HeightOverride=true, bOverride_WidthOverride=true, heightOverride=56, widthOverride=56}
   slot {class=HorizontalBoxSlot, properties {horizontalAlignment=HAlign_Right, verticalAlignment=VAlign_Center}}
  }
  InteractionIcon {
   widgetClassPath=&Image
   bIsVariable=true
   parent=InteractionIconSB
   widget=InteractionIcon
   properties {
    renderOpacity=0
    visibility=SelfHitTestInvisible
    brush {
     bIsDynamicallyLoaded=false
     drawAs=Image
     imageType=NoImage
     mirroring=NoMirror
     resourceName=None
     resourceObject=/Game/UI/Foundation/Textures/Icons/Checks/T-Icon-Check-64
     tiling=NoTile
     imageSize=<32 32>
     margin=<ltrb 0 0 0 0>
     outlineSettings {
      bUseBrushTransparency=false
      roundingType=HalfHeightRadius
      width=0
      color {colorUseRule=UseColor_Specified, specifiedColor=<rgba 0 0 0 0>}
      cornerRadii=<0 0 0 0>
     }
     tintColor {colorUseRule=UseColor_Specified, specifiedColor=<rgba 1 1 1 1>}
     uVRegion {bIsValid=false, max=<0 0>, min=<0 0>}
    }
   }
   slot {class=SizeBoxSlot, properties {horizontalAlignment=HAlign_Center, verticalAlignment=VAlign_Center}}
  }
  MenuAnchor_Actions {
   widgetClassPath=/Script/UMG.MenuAnchor
   bIsVariable=true
   parent=Overlay_1
   widget=MenuAnchor_Actions
   properties {menuClass=&/UserActionMenu.UserActionMenu_C, placement=MenuPlacement_RightLeftCenter}
   slot {class=OverlaySlot, properties {horizontalAlignment=HAlign_Right, verticalAlignment=VAlign_Center}}
  }
 ]
}
